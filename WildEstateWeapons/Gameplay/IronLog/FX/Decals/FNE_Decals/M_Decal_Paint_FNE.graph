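# Translucent deferred decal material for IronLog paint marks, built from the channels of the FNE decal texture.
# R and B tint red and blue over a cream base, while G forms a mask sampled with UVs rotated by a hash of the object pivot, varying each decal's orientation.
# Perlin noise biased by decal lifetime opacity and sharpened with CheapContrast multiplies that mask, so the paint dissolves patchily as it fades.
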Material /WildEstateWeapons/Gameplay/IronLog/FX/Decals/FNE_Decals/M_Decal_Paint_FNE
graph {
 properties {bAutomaticallySetUsageInEditor=false, blendMode=BLEND_Translucent, materialDomain=MD_DeferredDecal, usedShadingModels=Default Lit}
 nodes [
  Constant()
  Constant() {r=1}
  n1 = Constant3Vector() {constant=<rgba 0.7305 0.0529 0.0307 1>}
  n2 = TextureSample() {texture=&/T_IronLogDecal_FNE}
  n3 = Constant3Vector() {constant=<rgba 0.0343 0.1119 0.7305 1>}
  LinearInterpolate(
   A=Constant3Vector() {constant=<rgba 1 0.7991 0.6105 1>}
   B=Add(A=Multiply(A=@n1, B=@n2.R), B=Multiply(A=@n3, B=@n2.B))
   Alpha=Saturate(None=Add(A=@n2.R, B=@n2.R))
  )
  n4 = MaterialFunctionCall() {materialFunction=&ObjectPivotPoint}
  n5 = ComponentMask(None=MaterialFunctionCall(Vec3=Frac(None=@n4.Object Pivot Location)) {materialFunction=&MF_Hash}.Vec3) {r=true}
  n6 = MaterialFunctionCall(UVs=MaterialFunctionCall(UVs=TextureCoordinate(), "Rotation Angle (0-1)"=@n5) {materialFunction=&CustomRotator}.Rotated Values) {materialFunction=&ScaleUVsByCenter}
  n7 = MaterialFunctionCall(Texture Scale=AppendVector(A=ScalarParameter() {defaultValue=1, parameterName=UScale}, B=ScalarParameter() {defaultValue=1, parameterName=VScale})) {materialFunction=&ScaleUVsByCenter}
  n8 = DecalLifetimeOpacity()
  n9 = Saturate(None=Add(A=TextureSample(UVs=@n7.UVs) {texture=/Game/Global/Textures/Default/Noise/3xPerlin_M}.R, B=LinearInterpolate(Alpha=@n8.Opacity) {constA=-1}))
  Multiply(
   A=Multiply(A=TextureSample(UVs=@n6.UVs) {texture=&T_IronLogDecal_FNE}.G, B=@n6.0-1 mask)
   B=Multiply(A=MaterialFunctionCall(In=@n9) {materialFunction=&CheapContrast}.Result)
  )
  Multiply(A=@n5)
 ]
}
