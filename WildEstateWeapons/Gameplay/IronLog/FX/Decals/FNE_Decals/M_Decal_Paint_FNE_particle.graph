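# Translucent Niagara mesh-particle material for the IronLog weapon that fakes a projected paint decal.
# It maps the world position behind the translucency into instance space, giving size-adjustable UVs bounded by a sphere mask.
# Color blends particle color with red and blue texture-channel tints, shaded by scene normal against atmosphere light direction.
# Opacity combines particle alpha, a randomly rotated green-channel mask and Perlin noise that dissolves over the particle's later life.
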Material /WildEstateWeapons/Gameplay/IronLog/FX/Decals/FNE_Decals/M_Decal_Paint_FNE_particle
graph {
 properties {bAutomaticallySetUsageInEditor=false, bUsedWithNiagaraMeshParticles=true, blendMode=BLEND_Translucent, usedShadingModels=Default Lit}
 nodes [
  Constant3Vector() {constant=<rgba 1 0.7991 0.6105 1>}
  Constant()
  Constant() {r=1}
  n1 = ParticleColor()
  Multiply(A=@n1.RGB)
  n2 = Add(A=TransformPosition(None=MaterialFunctionCall() {materialFunction=&WorldPositionBehindTranslucency}.Result) {transformSourceType=TRANSFORMPOSSOURCE_World, transformType=TRANSFORMPOSSOURCE_Instance}) {constB=-0.5}
  n3 = ScalarParameter() {defaultValue=1, parameterName=Size}
  n4 = Divide(A=@n2, B=@n3)
  n5 = Constant() {r=0.5}
  n6 = ComponentMask(None=Subtract(A=Add(A=@n4, B=@n5), B=Divide(A=@n5, B=@n3) {desc=Scale the center offset by the scaling factor}) {desc=Re offset to maintain center}) {b=true, g=true}
  n7 = TextureSample(UVs=@n6) {texture=&/T_IronLogDecal_FNE}
  n8 = Multiply(A=Constant3Vector() {constant=<rgba 0.7305 0.0529 0.0307 1>}, B=@n7.R)
  n9 = Multiply(A=Constant3Vector() {constant=<rgba 0.0343 0.1119 0.7305 1>}, B=@n7.B)
  n10 = Add(A=@n7.R, B=@n7.R)
  n11 = DotProduct(A=ComponentMask(None=SceneTexture(UVs=@n6) {sceneTextureId=PPI_WorldNormal}.Color) {b=true, g=true, r=true}, B=SkyAtmosphereLightDirection())
  Multiply(
   A=LinearInterpolate(A=@n1.RGB, B=Add(A=@n8, B=@n9), Alpha=Saturate(None=@n10))
   B=Clamp(None=Power(Base=@n11) {constExponent=1}) {minDefault=0.5}
  )
  n12 = Constant() {r=-0.5}
  n13 = TextureSample(UVs=MaterialFunctionCall(UVs=@n6, Rotation Center=@n12, "Rotation Angle (0-1)"=ParticleRandom()) {materialFunction=&CustomRotator}.Rotated Values) {texture=&T_IronLogDecal_FNE}
  n14 = SphereMask(A=@n2, B=@n12, Radius=ScalarParameter() {defaultValue=0.5, parameterName=Radius}, Hardness=Constant() {r=0.5}) {attenuationRadius=1}
  n15 = ScalarParameter() {defaultValue=1, parameterName=UScale}
  n16 = ScalarParameter() {defaultValue=1, parameterName=VScale}
  n17 = TextureSample(UVs=MaterialFunctionCall(UVs=@n6, Texture Scale=AppendVector(A=@n15, B=@n16)) {materialFunction=&ScaleUVsByCenter}.UVs) {texture=/Game/Global/Textures/Default/Noise/3xPerlin_M}
  n18 = Multiply(A=ParticleRelativeTime()) {constB=2}
  n19 = LinearInterpolate(Alpha=Saturate(None=Subtract(A=@n18))) {constA=1, constAlpha=0, constB=-1}
  n20 = MaterialFunctionCall(In=Saturate(None=Add(A=@n17.R, B=@n19))) {materialFunction=&CheapContrast}
  Multiply(A=@n1.A, B=Multiply(A=Multiply(A=@n13.G, B=@n14), B=Multiply(A=@n20.Result)))
 ]
}
