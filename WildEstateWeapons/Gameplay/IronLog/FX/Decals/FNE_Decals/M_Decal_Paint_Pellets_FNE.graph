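# Translucent deferred decal material for paint pellet splats, by its name, in the IronLog weapon effects.
# A parameterized splash texture's green channel becomes a power-sharpened shape mask, multiplied by decal color alpha.
# Perlin noise, with center-scaled UVs set by UScale and VScale, is offset by decal lifetime opacity and contrast-adjusted.
# Multiplying that into the mask makes the splat dissolve in noisy patches rather than fade uniformly.
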Material /WildEstateWeapons/Gameplay/IronLog/FX/Decals/FNE_Decals/M_Decal_Paint_Pellets_FNE
graph {
 properties {bAutomaticallySetUsageInEditor=false, blendMode=BLEND_Translucent, materialDomain=MD_DeferredDecal, usedShadingModels=Default Lit}
 nodes [
  Constant()
  Constant() {r=1}
  n1 = TextureSampleParameter2D() {parameterName=Tex, texture=/Game/Effects/Fort_Effects/Textures/Liquid/T_FX_Splash_01}
  n2 = Saturate(None=Power(Base=OneMinus(None=@n1.G)) {constExponent=10})
  n3 = MaterialFunctionCall(Texture Scale=AppendVector(A=ScalarParameter() {defaultValue=1, parameterName=UScale}, B=ScalarParameter() {defaultValue=1, parameterName=VScale})) {materialFunction=&ScaleUVsByCenter}
  n4 = DecalLifetimeOpacity()
  n5 = Saturate(None=Add(A=TextureSample(UVs=@n3.UVs) {texture=/Game/Global/Textures/Default/Noise/3xPerlin_M}.R, B=LinearInterpolate(Alpha=@n4.Opacity) {constA=-1}))
  Multiply(
   A=Multiply(A=DecalColor().A, B=OneMinus(None=@n2))
   B=Multiply(A=MaterialFunctionCall(In=@n5) {materialFunction=&CheapContrast}.Result)
  )
 ]
}
