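# Translucent, lit paint-pellet decal material for IronLog weapon FX, built for Niagara mesh particles.
# Fakes decal projection by converting world position behind translucency into instance space, using YZ as centered, resizable UVs.
# The decal texture's green channel gets a random per-particle rotation, a spherical mask and particle alpha.
# Separate branches build a time-driven Perlin-noise dissolve and a scaled-UV mask; no material output connections appear in the graph.
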Material /WildEstateWeapons/Gameplay/IronLog/FX/Decals/FNE_Decals/M_Decal_Paint_Pellets_FNE_Particle
graph {
 properties {bAutomaticallySetUsageInEditor=false, bUsedWithNiagaraMeshParticles=true, blendMode=BLEND_Translucent, usedShadingModels=Default Lit}
 nodes [
  n1 = TransformPosition(None=MaterialFunctionCall() {materialFunction=&WorldPositionBehindTranslucency}.Result) {transformSourceType=TRANSFORMPOSSOURCE_World, transformType=TRANSFORMPOSSOURCE_Instance}
  ComponentMask(None=@n1) {b=true, g=true}
  Constant()
  Constant() {r=1}
  n2 = ScalarParameter() {defaultValue=1, parameterName=Size}
  n3 = Divide(A=Add(A=@n1) {constB=-0.5}, B=@n2)
  n4 = Constant() {r=0.5}
  n5 = ComponentMask(None=Subtract(A=Add(A=@n3, B=@n4), B=Divide(A=@n4, B=@n2) {desc=Scale the center offset by the scaling factor}) {desc=Re offset to maintain center}) {b=true, g=true}
  n6 = MaterialFunctionCall(
   UVs=@n5
   Texture Scale=AppendVector(A=ScalarParameter() {defaultValue=1, parameterName=UScale}, B=ScalarParameter() {defaultValue=1, parameterName=VScale})
  ) {materialFunction=&ScaleUVsByCenter}
  n7 = Subtract(A=Multiply(A=ParticleRelativeTime()) {constB=2})
  n8 = Add(
   A=Multiply(A=TextureSample(UVs=@n6.UVs) {texture=/Game/Global/Textures/Default/Noise/3xPerlin_M}.R, B=@n6.0-1 mask)
   B=LinearInterpolate(Alpha=Saturate(None=@n7)) {constA=1, constAlpha=0, constB=-1}
  )
  Multiply(A=MaterialFunctionCall(In=Saturate(None=@n8)) {materialFunction=&CheapContrast}.Result)
  n9 = Constant() {r=-0.5}
  n10 = MaterialFunctionCall(UVs=@n5, Rotation Center=@n9, "Rotation Angle (0-1)"=ParticleRandom()) {materialFunction=&CustomRotator}
  n11 = TextureSample(UVs=@n10.Rotated Values) {texture=&/T_IronLogDecal_FNE}
  n12 = SphereMask(A=@n5, B=@n9, Radius=ScalarParameter() {defaultValue=0.5, parameterName=Radius}, Hardness=Constant() {r=0.5}) {attenuationRadius=1}
  Multiply(A=Multiply(A=ParticleColor().A, B=Multiply(A=@n11.G, B=@n12)))
  Multiply(
   A=@n11.G
   B=MaterialFunctionCall(UVs=@n10.Rotated Values) {materialFunction=&ScaleUVsByCenter}.0-1 mask
  )
 ]
}
